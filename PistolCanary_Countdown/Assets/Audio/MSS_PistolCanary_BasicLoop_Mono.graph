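# Reusable mono MetaSound source that plays a supplied wave asset as a continuous loop, spatialized with the PistolCanary large POI attenuation.
# OnPlay or a deferred trigger variable starts the Wave Player at the source start time.
# On Nearly Finished writes that variable, retriggering playback as a feedback loop instead of using the player's loop setting.
# The Master Volume input scales the output.
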MetaSoundSource /PistolCanary_Countdown/Assets/Audio/MSS_PistolCanary_BasicLoop_Mono
inputs [
 Master Volume {type=Float}
 UE.Source.OnPlay {type=Trigger}
 UE.Source.StartTime.StartTime {type=Time}
 Wave Asset {type=WaveAsset}
]
interfaces [UE.OutputFormat.Mono:1.0, UE.Source.StartTime:1.0, UE.Source:1.1]
outputs ["UE.OutputFormat.Mono.Audio:0" {type=Audio}]
pages [Default {
  input_defaults {Master Volume=1, UE.Source.OnPlay=null, UE.Source.StartTime.StartTime=null, Wave Asset=null}
  nodes [
   n1 = VariableDeferredAccessor.Trigger(Variable=InitVariable.Trigger().Variable)
   TriggerAnyTriggerAny2_0 = "TriggerAny.Trigger Any (2)"(In 0=^UE.Source.OnPlay, In 1=@n1.Value)
   WavePlayer_0 = UE.Wave Player:Mono(Start Time=^UE.Source.StartTime.StartTime, Wave Asset=^Wave Asset, Play=@TriggerAnyTriggerAny2_0.Out)
   Multiply_0 = UE.Multiply:Audio by Float(AdditionalOperands=^Master Volume, PrimaryOperand=@WavePlayer_0.Out Mono)
   VariableMutator.Trigger(Variable=@n1.Variable, Value=@WavePlayer_0.On Nearly Finished)
  ]
  links [{dst="@UE.OutputFormat.Mono.Audio:0", src=Multiply_0.Out}]
 }]
properties {
 assetImportData=&self:AssetImportData
 attenuationSettings=/PistolCanary_Audio/Assets/Settings/ATT_PistolCanary_POI_Large
 bLooping=true
 duration=10000
 maxDistance=40000
 sampleRate=48000
}
