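# Looping stereo MetaSound source that plays one random wave from an input array, empty by default, using large point-of-interest attenuation.
# On Play, Random Get picks the wave, with its no-repeat count derived from array size, and triggers the stereo player.
# A deferred trigger variable feeds On Nearly Finished back into Play, replaying that wave without a new pick.
# Master Volume scales both output channels.
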MetaSoundSource /PistolCanary_Countdown/Assets/Audio/MSS_PistolCanary_LoopArray_Stereo
inputs [
 Master Volume {type=Float}
 UE.Source.OnPlay {type=Trigger}
 UE.Source.StartTime.StartTime {type=Time}
 Wave Asset {type=WaveAsset:Array}
]
interfaces [UE.OutputFormat.Stereo:1.0, UE.Source.StartTime:1.0, UE.Source:1.1]
outputs [
 "UE.OutputFormat.Stereo.Audio:0" {type=Audio}
 "UE.OutputFormat.Stereo.Audio:1" {type=Audio}
]
pages [Default {
  input_defaults {Master Volume=1, UE.Source.OnPlay=null, UE.Source.StartTime.StartTime=null, Wave Asset []}
  nodes [
   ArrayRandomGet_0 = Array.Random Get:WaveAsset:Array(
    Next=^UE.Source.OnPlay
    In Array=^Wave Asset
    No Repeats=UE.Divide:Int32(PrimaryOperand=Array.Num:WaveAsset:Array(Array=^Wave Asset).Num).Out
   ) {also {defaults {Weights []}}}
   n1 = VariableDeferredAccessor.Trigger(Variable=InitVariable.Trigger().Variable)
   TriggerAnyTriggerAny2_0 = "TriggerAny.Trigger Any (2)"(In 0=@ArrayRandomGet_0.On Next, In 1=@n1.Value)
   WavePlayer_0 = UE.Wave Player:Stereo(Start Time=^UE.Source.StartTime.StartTime, Wave Asset=@ArrayRandomGet_0.Value, Play=@TriggerAnyTriggerAny2_0.Out)
   Multiply_0 = UE.Multiply:Audio by Float(AdditionalOperands=^Master Volume, PrimaryOperand=@WavePlayer_0.Out Left)
   Multiply_1 = UE.Multiply:Audio by Float(AdditionalOperands=^Master Volume, PrimaryOperand=@WavePlayer_0.Out Right)
   VariableMutator.Trigger(Variable=@n1.Variable, Value=@WavePlayer_0.On Nearly Finished)
  ]
  links [{dst="@UE.OutputFormat.Stereo.Audio:0", src=Multiply_0.Out}, {dst="@UE.OutputFormat.Stereo.Audio:1", src=Multiply_1.Out}]
 }]
properties {
 assetImportData=&self:AssetImportData
 attenuationSettings=/PistolCanary_Audio/Assets/Settings/ATT_PistolCanary_POI_Large
 bLooping=true
 duration=10000
 maxDistance=40000
 numChannels=2
 outputFormat=Stereo
 sampleRate=48000
}
