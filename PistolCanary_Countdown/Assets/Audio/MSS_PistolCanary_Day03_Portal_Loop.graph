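# Stereo MetaSound source preset that plays the looping ambient sound for the PistolCanary Countdown Day 03 portal.
# It inherits its graph from the MSS_PistolCanary_LoopArray_Stereo parent, overriding only inputs: the wave array holds the single Day 03 portal loop, with a master volume control.
# It loops under the Fort_Ambient_Athena sound class, with spatial falloff from the shared PortalLoop attenuation settings.
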MetaSoundSource /PistolCanary_Countdown/Assets/Audio/MSS_PistolCanary_Day03_Portal_Loop
preset_parent=&/MSS_PistolCanary_LoopArray_Stereo
inputs [
 Master Volume {type=Float}
 UE.Source.OnPlay {type=Trigger}
 UE.Source.StartTime.StartTime {type=Time}
 Wave Asset {type=WaveAsset:Array}
]
interfaces [UE.OutputFormat.Stereo:1.0, UE.Source.StartTime:1.0, UE.Source:1.1]
outputs [
 "UE.OutputFormat.Stereo.Audio:0" {type=Audio}
 "UE.OutputFormat.Stereo.Audio:1" {type=Audio}
]
pages [Default {input_defaults {
   Master Volume=1
   UE.Source.OnPlay=null
   UE.Source.StartTime.StartTime=null
   Wave Asset [/PistolCanary_Countdown/Assets/Audio/wav/PistolCanary_Day_03_Portal_Loop]
  }}]
properties {
 assetImportData=&self:AssetImportData
 attenuationSettings=/PistolCanary_Countdown/Assets/Audio/Settings/ATT_PistolCanary_PortalLoop
 bLooping=true
 duration=10000
 maxDistance=4150
 numChannels=2
 outputFormat=Stereo
 sampleRate=48000
 soundClassObject=&Fort_Ambient_Athena
}
