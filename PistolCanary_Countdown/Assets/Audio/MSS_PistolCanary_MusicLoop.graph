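# Stereo MetaSound source that plays looping, attenuated POI music for PistolCanary Countdown, limited to one instance.
# An exposed Wave Asset input feeds a wave player started by On Play.
# The player's On Nearly Finished trigger returns through a deferred trigger variable to restart playback, creating the loop.
# Both channels are scaled by Master Volume multiplied by a modulator value before reaching the output.
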MetaSoundSource /PistolCanary_Countdown/Assets/Audio/MSS_PistolCanary_MusicLoop
inputs [
 Master Volume {type=Float}
 UE.Source.OnPlay {type=Trigger}
 Wave Asset {type=WaveAsset}
]
interfaces [UE.OutputFormat.Stereo:1.0, UE.Source:1.1]
outputs [
 "UE.OutputFormat.Stereo.Audio:0" {type=Audio}
 "UE.OutputFormat.Stereo.Audio:1" {type=Audio}
]
pages [Default {
  input_defaults {Master Volume=0.25, UE.Source.OnPlay=null, Wave Asset=null}
  nodes [
   n1 = UE.Multiply:Float(PrimaryOperand=^Master Volume, AdditionalOperands=Modulation.GetModulatorValue().Out)
   n2 = VariableDeferredAccessor.Trigger(Variable=InitVariable.Trigger().Variable)
   TriggerAnyTriggerAny2_0 = "TriggerAny.Trigger Any (2)"(In 0=^UE.Source.OnPlay, In 1=@n2.Value)
   WavePlayer_0 = UE.Wave Player:Stereo(Wave Asset=^Wave Asset, Play=@TriggerAnyTriggerAny2_0.Out)
   Multiply_0 = UE.Multiply:Audio by Float(AdditionalOperands=@n1.Out, PrimaryOperand=@WavePlayer_0.Out Left)
   Multiply_1 = UE.Multiply:Audio by Float(AdditionalOperands=@n1.Out, PrimaryOperand=@WavePlayer_0.Out Right)
   VariableMutator.Trigger(Variable=@n2.Variable, Value=@WavePlayer_0.On Nearly Finished)
  ]
  links [{dst="@UE.OutputFormat.Stereo.Audio:0", src=Multiply_0.Out}, {dst="@UE.OutputFormat.Stereo.Audio:1", src=Multiply_1.Out}]
 }]
properties {
 assetImportData=&self:AssetImportData
 attenuationSettings=/PistolCanary_Countdown/Assets/Audio/Settings/ATT_PistolCanary_POI_Music
 bLooping=true
 bOverrideConcurrency=true
 duration=10000
 maxDistance=10500
 numChannels=2
 outputFormat=Stereo
 sampleRate=48000
 soundClassObject=/Game/Sounds/FortSoundClassesAndModes/Fort_Music_POI_PSM
 concurrencyOverrides {maxCount=1}
}
