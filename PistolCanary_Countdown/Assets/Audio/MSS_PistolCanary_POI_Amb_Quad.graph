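# Quad MetaSound source that plays a looping point-of-interest ambience bed for PistolCanary, using large POI attenuation.
# A quad Wave Player starts on play and retriggers itself through a deferred trigger variable fed by its Nearly Finished output.
# Its four channels pass through the MSP_AzimuthPannerQuad patch, driven by listener azimuth with MasterVolume as gain, to the front and side outputs.
# The wave asset is an input with no default assigned.
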MetaSoundSource /PistolCanary_Countdown/Assets/Audio/MSS_PistolCanary_POI_Amb_Quad
inputs [
 MasterVolume {type=Float}
 UE.Listener.Orientation.Azimuth {type=Float}
 UE.Listener.Orientation.Elevation {type=Float}
 UE.Source.OnPlay {type=Trigger}
 Wave Asset {type=WaveAsset}
]
interfaces [UE.Listener.Orientation:1.0, UE.OutputFormat.Quad:1.0, UE.Source:1.1]
outputs [
 "UE.OutputFormat.Quad.Audio:0" {type=Audio}
 "UE.OutputFormat.Quad.Audio:1" {type=Audio}
 "UE.OutputFormat.Quad.Audio:4" {type=Audio}
 "UE.OutputFormat.Quad.Audio:5" {type=Audio}
]
pages [Default {
  input_defaults {MasterVolume=1, UE.Listener.Orientation.Azimuth=117.031, UE.Listener.Orientation.Elevation=0, UE.Source.OnPlay=null, Wave Asset=null}
  nodes [
   n1 = VariableDeferredAccessor.Trigger(Variable=InitVariable.Trigger().Variable)
   TriggerAnyTriggerAny2_0 = "TriggerAny.Trigger Any (2)"(In 0=^UE.Source.OnPlay, In 1=@n1.Value)
   WavePlayer_0 = UE.Wave Player:Quad(Wave Asset=^Wave Asset, Play=@TriggerAnyTriggerAny2_0.Out)
   MSP_AzimuthPannerQuad_0 = &MSP_AzimuthPannerQuad(
    Gain=^MasterVolume
    Azimuth=^UE.Listener.Orientation.Azimuth
    In Front Left=@WavePlayer_0.Out Front Left
    In Front Right=@WavePlayer_0.Out Front Right
    In Side Left=@WavePlayer_0.Out Side Left
    In Side Right=@WavePlayer_0.Out Side Right
   )
   VariableMutator.Trigger(Variable=@n1.Variable, Value=@WavePlayer_0.On Nearly Finished)
  ]
  links [
   {dst="@UE.OutputFormat.Quad.Audio:0", src=MSP_AzimuthPannerQuad_0.Out Front Left}
   {dst="@UE.OutputFormat.Quad.Audio:1", src=MSP_AzimuthPannerQuad_0.Out Front Right}
   {dst="@UE.OutputFormat.Quad.Audio:4", src=MSP_AzimuthPannerQuad_0.Out Side Left}
   {dst="@UE.OutputFormat.Quad.Audio:5", src=MSP_AzimuthPannerQuad_0.Out Side Right}
  ]
 }]
properties {
 assetImportData=&self:AssetImportData
 attenuationSettings=/PistolCanary_Audio/Assets/Settings/ATT_PistolCanary_POI_Large
 bLooping=true
 duration=10000
 maxDistance=40000
 numChannels=4
 outputFormat=Quad
 sampleRate=48000
 soundClassObject=&Fort_Events
}
