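# Sound cue that plays the PistolCanary timer tick, named for the 10-second tick of the countdown.
# A single wave player feeds the output directly at reduced volume, using the 60-second countdown attenuation asset for spatial falloff.
# It routes to the important in-game mission sound class and overrides concurrency to cap simultaneous instances at two.
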SoundCue /PistolCanary_Countdown/Assets/Audio/PistolCanary_Countdown_10sTick_Cue
root=WavePlayer_1
nodes [{id=WavePlayer_1, properties {soundWaveAssetPtr=&/PistolCanary_Timer_Tick}}]
properties {
 attenuationSettings=&/PistolCanary_Countdown_60s_Att
 bOverrideConcurrency=true
 maxDistance=11500
 soundClassObject=&Fort_Mission_Ingame_Important_PSM
 volumeMultiplier=0.35
 concurrencyOverrides {maxCount=2}
}
