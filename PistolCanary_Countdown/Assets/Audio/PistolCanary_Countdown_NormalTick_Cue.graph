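# Sound cue for the normal countdown tick in the PistolCanary Countdown feature.
# A single wave player plays the Timer_Tick_02 wave straight to the output at reduced volume.
# It uses the countdown attenuation asset and the Fort_Events sound class, with a concurrency override that caps simultaneous instances.
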SoundCue /PistolCanary_Countdown/Assets/Audio/PistolCanary_Countdown_NormalTick_Cue
root=WavePlayer_2
nodes [{id=WavePlayer_2, properties {soundWaveAssetPtr=&/PistolCanary_Timer_Tick_02}}]
properties {
 attenuationSettings=&/PistolCanary_Countdown_Att
 bOverrideConcurrency=true
 maxDistance=4200
 soundClassObject=&Fort_Events
 volumeMultiplier=0.5
 concurrencyOverrides {maxCount=2}
}
