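# Translucent, volumetric-directional lit material for contrail ribbons, usable on Niagara ribbons, beam trails and sprites, with a cloudy, soft-edged look.
# Panning cloud textures on UV1 are offset by dynamic PanAdd and warped by a normal map more strongly as particle alpha drops.
# HeightLerp erodes a cross-width profile with cloud noise, multiplied by particle alpha, sine end-cap fade and quality-switched depth fade.
# A sphere-masked GlowColor, scaled by dynamic GlowBrightness, feeds emissive adjustments.
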Material /PistolCanary_Countdown/Assets/FX/M_SimpleContrailRibbon_Lit
graph {
 properties {
  bUsedWithBeamTrails=true
  bUsedWithNanite=true
  bUsedWithNiagaraRibbons=true
  bUsedWithParticleSprites=true
  blendMode=BLEND_Translucent
  pixelDepthOffsetMode=PDOM_Legacy
  translucencyLightingMode=TLM_VolumetricDirectional
  usedShadingModels=Default Lit
 }
 nodes [
  n1 = VectorParameter() {parameterName=GlowColor, defaultValue=<rgba 3 0.1899 0 1>}
  n2 = Power(Base=SphereMask(A=TextureCoordinate(), B=Constant2Vector() {g=0.5}, Radius=Constant() {r=0.5}) {hardnessPercent=0})
  n3 = DynamicParameter() {defaultValue=<rgba 1 0 1 1>, paramNames [PanAdd, GlowBrightness, End Cap Fade, Param4]}
  MaterialFunctionCall(Emissive=Multiply(A=Multiply(A=@n1.RGB, B=@n2), B=@n3.GlowBrightness)) {materialFunction=&MF_FX_EmissiveAdjustments}
  n4 = Subtract(B=ComponentMask(None=TextureCoordinate()) {g=true}) {constA=0.5}
  n5 = MaterialFunctionCall!(Side 1 Length=Multiply(A=Abs(None=@n4)) {constB=2}) {materialFunction=/Engine/Functions/Engine_MaterialFunctions02/Math/pythagoreanTheorum}
  n6 = Multiply(
   A=TextureCoordinate() {coordinateIndex=1}
   B=ComponentMask(None=VectorParameter() {parameterName=UV Tile_1, defaultValue=<rgba 0.4 0.3 0 0>}.RGB) {g=true, r=true}
  )
  n7 = Multiply(
   A=OneMinus(None=ParticleColor().A)
   B=ComponentMask(None=TextureSample(UVs=@n6) {samplerType=SAMPLERTYPE_Normal, texture=/Game/Effects/Fort_Effects/BluGlow/testNormals}.RGB) {g=true, r=true}
  ) {constA=1}
  n8 = ScalarParameter() {defaultValue=0.05, parameterName=UV Dist}
  n9 = Add(
   A=TextureCoordinate() {coordinateIndex=1}
   B=Multiply(
    A=DynamicParameter() {defaultValue=<rgba 1 0 1 1>, paramNames [PanAdd, GlowBrightness, End Cap Fade, Param4]}.PanAdd
    B=ScalarParameter() {defaultValue=0.5, parameterName=XY Shift}
   )
  )
  n10 = ComponentMask(None=VectorParameter() {parameterName=UV Tile, defaultValue=<rgba 0.1 0.3 0 0>}.RGB) {g=true, r=true}
  n11 = Panner(
   Coordinate=Add(A=Multiply(A=@n7, B=@n8), B=Multiply(A=@n9, B=@n10))
   Speed=ComponentMask(None=VectorParameter() {parameterName=PanningSpeed}.RGB) {g=true, r=true}
  )
  n12 = MaterialFunctionCall(
   Transition Phase=Multiply(A=@n5.Side 2 Length, B=@n5.Side 2 Length)
   Height Texture=LinearInterpolate(Alpha=TextureSample(UVs=@n11) {samplerType=SAMPLERTYPE_LinearColor, texture=/Game/Global/Textures/Default/Tileable_Patterns/Cloud_Diffuse}.R) {constA=-1}
  ) {materialFunction=&HeightLerp}
  n13 = ParticleColor()
  n14 = Sine(None=ComponentMask(None=TextureCoordinate()) {r=true}) {period=2}
  n15 = ScalarParameter() {defaultValue=25, parameterName=DepthFadeDistance}
  Multiply(
   A=Multiply(A=Multiply(A=@n12.Alpha, B=@n13.A), B=Power(Base=@n14, Exp=@n3.End Cap Fade))
   B=QualitySwitch(Default=DepthFade(FadeDistance=@n15), Low=Constant() {r=1})
  )
  ParticleColor()
  TextureSample(UVs=@n11) {samplerType=SAMPLERTYPE_Normal, texture=/Game/Global/Textures/Default/Tileable_Patterns/Cloud_normals}
 ]
}
