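# Translucent unlit master material for the PistolCanary countdown hologram effect, built on material attributes.
# All shading logic lives in a single call to the MF_PistolCanaryCountdownHologram material function.
# It enables responsive AA and is flagged for use on instanced static meshes.
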Material /PistolCanary_Countdown/Assets/FX/Materials/M_Countdown_Master_F_PistolCanary
graph {
 properties {
  bEnableResponsiveAA=true
  bUseMaterialAttributes=true
  bUsedWithInstancedStaticMeshes=true
  blendMode=BLEND_Translucent
  pixelDepthOffsetMode=PDOM_Legacy
  shadingModel=MSM_Unlit
  usedShadingModels=Unlit
 }
 nodes [MaterialFunctionCall() {materialFunction=&/MF_PistolCanaryCountdownHologram}]
}
