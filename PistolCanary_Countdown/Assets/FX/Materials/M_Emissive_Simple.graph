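# Simple unlit emissive material for FX in PistolCanary_Countdown, set up for Niagara mesh particles.
# Its only node is Particle Color, with no textures or parameters, so the particle system supplies the color.
# Automatic usage flagging in the editor is off, and Niagara mesh particles is the usage explicitly enabled.
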Material /PistolCanary_Countdown/Assets/FX/Materials/M_Emissive_Simple
graph {
 properties {bAutomaticallySetUsageInEditor=false, bUsedWithNiagaraMeshParticles=true, shadingModel=MSM_Unlit, usedShadingModels=Unlit}
 nodes [ParticleColor()]
}
