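# Unlit, two-sided translucent energy material for Niagara ribbon and mesh particles, with SplineThicken setting width from a parameter, optionally scaled by dynamic parameter.
# Panning plasma-coil texture, distorted by inky smoke noise, is tinted by MainColor and particle color for emissive.
# Opacity combines panning smoke trail and blurred noise, sine-tapered along U.
# Particle alpha, view angle, depth fade on deferred paths and optional near-camera fade modulate opacity.
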Material /PistolCanary_Countdown/Assets/FX/Materials/M_SplineThicken_Energy
graph {
 properties {
  bUsedWithNiagaraMeshParticles=true
  bUsedWithNiagaraRibbons=true
  blendMode=BLEND_Translucent
  pixelDepthOffsetMode=PDOM_Legacy
  shadingModel=MSM_Unlit
  twoSided=true
  usedShadingModels=Unlit
 }
 nodes [
  Constant() {r=20}
  n1 = TextureCoordinate()
  n2 = ComponentMask(None=VectorParameter() {parameterName=UV_Scale, defaultValue=<rgba 1 1 0 0>}.RGB) {g=true, r=true}
  n3 = TextureCoordinate()
  n4 = Multiply(A=TextureSample(UVs=Panner(Coordinate=@n3) {speedX=0.1}) {texture=&T_Inky_Smoke_Tile}.R) {constB=0.15}
  n5 = Time()
  n6 = Multiply(A=@n5, B=ScalarParameter() {defaultValue=-0.25, parameterName=Speed Direction}) {constB=-1.5}
  n7 = Panner(Coordinate=Add(A=@n4, B=Multiply(A=@n1, B=@n2)), Time=@n6) {speedX=-0.8}
  n8 = Multiply(
   A=Desaturation(
    None=VectorParameter() {parameterName=MainColor, defaultValue=<rgba 1.5 0 1.0258 1>}.RGB
    Fraction=ScalarParameter() {parameterName=DefaultColorDes}
   )
   B=Power(
    Base=TextureSample(UVs=@n7) {samplerType=SAMPLERTYPE_LinearGrayscale, texture=/Game/Effects/Fort_Effects/Emotes/UltraLight/Texture/T_PlasmaCoils}.R
    Exp=ScalarParameter() {defaultValue=1.5, parameterName=Power}
   ) {constExponent=1.5}
  )
  n9 = TextureCoordinate()
  n10 = TextureSample(UVs=Panner(Coordinate=Add(A=@n9, B=@n4), Time=@n6) {speedX=-1}) {samplerType=SAMPLERTYPE_Grayscale, texture=&T_Smoke_Trail_Soft}
  n11 = Multiply(A=Constant3Vector() {constant=<rgba 0 0 0 1>}, B=@n10.R)
  MaterialFunctionCall(Emissive=Multiply(A=ParticleColor().RGB, B=Add(A=@n8, B=@n11))) {materialFunction=&MF_FX_EmissiveAdjustments}
  n12 = DynamicParameter() {paramNames [Width,Time,Param3,Param4]}
  n13 = Multiply(
   A=ScalarParameter() {defaultValue=25, parameterName=Width}
   B=LinearInterpolate(B=@n12.Width, Alpha=ScalarParameter() {parameterName=UseCustomWidth}) {constA=1}
  )
  MaterialFunctionCall(WidthBase=@n13, WidthTip=@n13, Expand U or V UV Channel=StaticBoolParameter() {parameterName="Expand U or V UV Channel (SB)"}) {materialFunction=&SplineThicken}
  Multiply(A=@n5, B=NamedRerouteUsage())
  NamedRerouteDeclaration(None=@n12.Time) {name=OUT_Time, nodeColor=<rgba 0.8516 0 1 1>}
  n14 = MaterialFunctionCall(Distance=ScalarParameter() {defaultValue=200, parameterName=Camera Fade Distance}) {materialFunction=&Near_Camera_Fade}
  n15 = Add(A=@n4, B=TextureCoordinate())
  n16 = Multiply(
   A=TextureSample(UVs=Panner(Coordinate=@n15, Time=@n6) {speedX=-1.5}) {samplerType=SAMPLERTYPE_Grayscale, texture=&LowResBlurredNoise}.R
   B=@n10.R
  )
  n17 = ComponentMask(None=TextureCoordinate()) {r=true}
  n18 = Multiply(A=ParticleColor().A, B=Multiply(A=Add(A=@n10.R, B=@n16) {constB=0}, B=Sine(None=@n17) {period=2}))
  n19 = DepthFade(Opacity=@n18, FadeDistance=ScalarParameter() {defaultValue=10, parameterName=FadeDistance}) {fadeDistanceDefault=10}
  n20 = MaterialFunctionCall("FadeSensitivity (optional)"=ScalarParameter() {parameterName="FadeSensitivity (optional) (S)"}) {materialFunction=&FadeBasedOnViewAngle}
  n21 = Multiply(A=ShadingPathSwitch(Default=@n19, Deferred=@n19, Forward=@n18, Mobile=@n18), B=Saturate(None=Multiply(A=@n20.Result) {constB=10}))
  StaticSwitchParameter(True=Multiply(A=Multiply(A=@n14.Result, B=@n14.Result), B=@n21), False=@n21) {parameterName=Use Camera Distance Fade}
  TextureCoordinate() {uTiling=3}
 ]
}
