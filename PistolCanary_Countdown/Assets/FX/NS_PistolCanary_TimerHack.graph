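# Niagara effect for the PistolCanary Countdown timer hack: a GPU burst of short-lived green pixel sprites, plus a spawn rate decaying over the emitter loop.
# Particles randomly split into large steady or small jittering sprites, with randomized material parameters and camera offsets.
# A raised box location, reapplied on update, uses a scratch input to derive yaw from camera and system positions.
# User.IsLogo switches that box between logo and wide-strip proportions.
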NiagaraSystem /PistolCanary_Countdown/Assets/FX/NS_PistolCanary_TimerHack
dependencies {
 usedDataInterfaces [&NiagaraDataInterfaceCurve, &NiagaraDataInterfaceCamera]
 usedDynamicInputs [&OrientMeshToVector]
 usedModules [&SystemState]
}
emitters {Pixels {
  simTarget=GPUComputeSim
  emitterUpdateScript [
   &EmitterState()
   &SpawnBurst_Instantaneous(Spawn Count=35, Use Spawn Probability=false, UseLoopCountLimit=false)
   &SpawnRate(
    SpawnRate=&FloatFromCurve(
     FloatCurve=DataInterface() {propertyValues {Curve {keys [
         {
          arriveTangent=-16.7652
          arriveTangentWeight=0.8266
          interpMode=RCIM_Cubic
          leaveTangent=-16.7652
          leaveTangentWeight=0.8266
          tangentMode=RCTM_User
          tangentWeightMode=RCTWM_WeightedBoth
          time=0
          value=1
         }
         {
          arriveTangent=-0.0137
          arriveTangentWeight=0.86
          interpMode=RCIM_Cubic
          leaveTangent=-0.0137
          leaveTangentWeight=0.86
          tangentMode=RCTM_User
          tangentWeightMode=RCTWM_WeightedBoth
          time=1
          value=0
         }
        ]}}}
     CurveIndex=$Emitter.NormalizedLoopAge
     Scale Curve=300
    )
    Use Spawn Probability=false
   ) {module=SpawnRate001}
  ]
  particleSpawnScript [
   /SetVariables(Particles.VisibilityTag=&RandomRangeInt(Minimum=0, Maximum=1)) {isSetParameters=true}
   /SetVariables(Particles.Is Transparent one:NiagaraBool=&SetBoolByIntComparison(A=$Particles.VisibilityTag, B=1, Comparison Type=::NewEnumerator4)) {isSetParameters=true}
   &InitializeParticle(
    Color=<rgba 0 0.5 0 1>
    Lifetime Max=0.5
    Lifetime Min=0.2
    Lifetime Mode=::NewEnumerator1
    Mass Mode=::NewEnumerator3
    Sprite Size Mode=::NewEnumerator3
    Uniform Sprite Size=&MakeCustomFloatFromBool(Bool=$Particles.Is Transparent one, True Float=100, False Float=25)
    UsePositionOffset=false
    Write Material Random=true
   )
   &ApplyInitialForces()
   &DynamicMaterialParameters(
    Index 0 Param 1=&RandomRangeFloat(Minimum=0.1, Maximum=0.3)
    Index 0 Param 2=&RandomRangeFloat(Minimum=1.5, Maximum=2.5)
    Index 0 Param 3=&RandomRangeFloat(Minimum=0.1, Maximum=0.5)
    Index 0 Param 4=1
    Module.Param0WriteEnabled=true
    Module.Param1WriteEnabled=true
    Module.Param2WriteEnabled=true
    Module.Param3WriteEnabled=true
   )
   &CameraOffset(Camera Offset Amount=&RandomRangeFloat(Minimum=-25, Maximum=25))
  ]
  particleUpdateScript [
   &ParticleState()
   &SolveForcesAndVelocity(Clamp Velocity=false, Limit Acceleration=false)
   &ShapeLocation(
    Box Size=&MakeCustomVectorFromBool(Bool=$User.IsLogo, True Vector=<10 1000 800>, False Vector=<10 1700 250>)
    MaskSpawnGroup=false
    Offset=<0 0 900>
    Offset Mode=::NewEnumerator0
    Override Seed=false
    Rotation Angle:NiagaraFloat=&self:ScratchDynamicInput(
     Position:NiagaraPosition=&ReturnCameraPosition(Coordinate Space:ENiagaraCoordinateSpace=::Simulation, Camera Query:NiagaraDataInterfaceCamera=DataInterface())
     Position001:NiagaraPosition=/Niagara/DynamicInputs/Helpers/SimulationPosition()
    )
    Rotation Axis=<0 0 1>
    Rotation Mode=::NewEnumerator1
    Shape Primitive=::NewEnumerator2
    Surface Only=false
    UseOwnerScale=true
   )
   &JitterPosition(
    Jitter Amount:NiagaraFloat=&MakeCustomFloatFromBool(Bool=$Particles.Is Transparent one, True Float=0, False Float=100)
    Jitter Delay:NiagaraFloat=&FloatFromCurve(
     FloatCurve=DataInterface() {propertyValues {Curve {keys [
         {
          arriveTangent=0.2256
          arriveTangentWeight=0.8018
          interpMode=RCIM_Cubic
          leaveTangent=0.2256
          leaveTangentWeight=0.8018
          tangentMode=RCTM_User
          tangentWeightMode=RCTWM_WeightedBoth
          time=0
          value=0
         }
         {
          arriveTangent=0.3528
          arriveTangentWeight=0.8252
          interpMode=RCIM_Cubic
          leaveTangent=0.3528
          leaveTangentWeight=0.8252
          tangentMode=RCTM_User
          tangentWeightMode=RCTWM_WeightedBoth
          time=1
          value=1
         }
        ]}}}
     CurveIndex=&RandomRangeFloat(Minimum=0, Maximum=1)
     Scale Curve=0.2
    )
   )
  ]
  rendererClasses [&NiagaraSpriteRendererProperties]
  renderers [
   {rendererClass=&NiagaraSpriteRendererProperties}
   {rendererClass=&NiagaraSpriteRendererProperties, rendererIndex=1}
  ]
 }}
summary {
 systemName=NS_PistolCanary_TimerHack
 emitters [{emitterName=Pixels, simTarget=GPUComputeSim, rendererClasses [&NiagaraSpriteRendererProperties]}]
 userVariables [User.IsLogo {type=NiagaraBool}]
}
