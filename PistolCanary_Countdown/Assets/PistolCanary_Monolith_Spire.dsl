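; Monolith spire actor for the Pistol Canary countdown, configured as an outer or central spire by an OuterSpire flag.
; The construction script scales outer spires three times larger and disables auto-activation of their camera shake and force feedback.
; On BeginPlay, outer spires destroy those feedback components, while other spires set a very large net cull distance.
; A SelfDestruct event destroys the actor, and the OnRep_TurnOnEffectsClient function is empty.
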(event EventBeginPlay
  (if (Variables|Default|GetOuterSpire)
    (Components|DestroyComponent (Variables|Default|GetCameraShakeSource))
    (Components|DestroyComponent (Variables|Default|GetForceFeedback))
    (else
      (Variables|Replication|SetNetCullDistanceSquared 1666633333.5))))

(event Custom|SelfDestruct
  (Actor|DestroyActor))

(fn OnRep_TurnOnEffectsClient)

(fn ConstructionScript
  (bind _staticmeshcomponent (Variables|StaticMeshActor|GetStaticMeshComponent))
  (if (Variables|Default|GetOuterSpire)
    (Transformation|SetRelativeScale3D _staticmeshcomponent "3.000000,3.000000,3.000000")
    (Components|Activation|SetAutoActivate (Variables|Default|GetCameraShakeSource) false)
    (Components|Activation|SetAutoActivate (Variables|Default|GetForceFeedback) false)
    (else
      (Transformation|SetRelativeScale3D _staticmeshcomponent "1.000000,1.000000,1.000000"))))
